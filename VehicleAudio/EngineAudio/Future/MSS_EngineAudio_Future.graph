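# Looping stereo MetaSound for the Future vehicle engine, driven mainly by normal and bonus speed.
# A constant low-speed loop mixes with medium and high loops managed by speed crossfade segments.
# All three share a pitch from a smoothed, piecewise cubic speed curve plus bonus speed.
# Bonus speed starts and stops a supersonic layer with speed-scaled volume, and character affiliation sets the final gain of both buses.
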MetaSoundSource /VehicleAudio/EngineAudio/Future/MSS_EngineAudio_Future
inputs [
 Character.Affiliation {type=Enum:FortAudioAffiliation}
 Character.Health {type=Float}
 Character.UniqueID {type=Int32}
 Fort.Vehicle.Acceleration {type=Float}
 Fort.Vehicle.BlowoffStrength {type=Float}
 Fort.Vehicle.BonusSpeed {type=Float}
 Fort.Vehicle.EnableShifting {type=Bool}
 Fort.Vehicle.Gear {type=Int32}
 Fort.Vehicle.IsGrounded {type=Bool}
 Fort.Vehicle.Load {type=Float}
 Fort.Vehicle.LowFuel {type=Bool}
 Fort.Vehicle.NormalSpeed {type=Float}
 Fort.Vehicle.OnBlowoff {type=Trigger}
 Fort.Vehicle.OnDownshift {type=Trigger}
 Fort.Vehicle.OnUpshift {type=Trigger}
 Fort.Vehicle.Pitch {type=Float}
 Fort.Vehicle.RPM {type=Float}
 Fort.Vehicle.Throttle {type=Float}
 UE.Source.OnPlay {type=Trigger}
]
interfaces [Fort.Character:1.0, Fort.Vehicle:1.0, UE.OutputFormat.Stereo:1.0, UE.Source:1.1]
outputs [
 "UE.OutputFormat.Stereo.Audio:0" {type=Audio}
 "UE.OutputFormat.Stereo.Audio:1" {type=Audio}
]
pages [Default {
  input_defaults {
   Character.Affiliation=0
   Character.Health=0
   Character.UniqueID=0
   Fort.Vehicle.Acceleration=0
   Fort.Vehicle.BlowoffStrength=0
   Fort.Vehicle.BonusSpeed=0
   Fort.Vehicle.EnableShifting=null
   Fort.Vehicle.Gear=0
   Fort.Vehicle.IsGrounded="<MetaSoundBuilderResult.SUCCEEDED: 0>"
   Fort.Vehicle.Load=0
   Fort.Vehicle.LowFuel=null
   Fort.Vehicle.NormalSpeed=0
   Fort.Vehicle.OnBlowoff=null
   Fort.Vehicle.OnDownshift=null
   Fort.Vehicle.OnUpshift=null
   Fort.Vehicle.Pitch=0
   Fort.Vehicle.RPM=0
   Fort.Vehicle.Throttle=0
   UE.Source.OnPlay=null
  }
  nodes [
   n1 = Fortnite.CrossfadeSegmentManager:Float(In=^Fort.Vehicle.NormalSpeed) {also {defaults {Max=1.2, MaxRange=0, Min=0.0915, MinRange=0.5561}}}
   n2 = Fortnite.CrossfadeSegmentManager:Float(In=^Fort.Vehicle.NormalSpeed) {also {defaults {Max=0.6477, MaxRange=0.5561, Min=-0.1, MinRange=0}}}
   n3 = UE.InterpTo:Audio(Target=MapRange.MapRange:Float(In=^Fort.Vehicle.BonusSpeed) {also {defaults {Out Range B=0.5}}}.Out Value) {also {defaults {Interp Time=0.15}}}
   n4 = UE.Power:Float(PrimaryOperand=UE.Subtract:Float(AdditionalOperands=^Fort.Vehicle.NormalSpeed) {also {defaults {PrimaryOperand=1}}}.Out) {also {defaults {AdditionalOperands=3}}}
   TriggerOnValueChange_0 = UE.Trigger On Value Change:Float(Value=^Fort.Vehicle.NormalSpeed)
   n5 = TriggerCompare.Clamp:Float(A=^Fort.Vehicle.NormalSpeed, Compare=@TriggerOnValueChange_0.Trigger) {also {defaults {B=0.5, Type=5}}}
   TriggerOnce_0 = UE.Trigger Once(Trigger In=@n5.False, Reset=@n5.True)
   TriggerOnce_1 = UE.Trigger Once(Reset=@n5.False, Trigger In=@n5.True)
   TriggerRouteTriggerRouteFloat2_1 = "TriggerRoute.Trigger Route (Float, 2):Float"(
    Value 1=UE.Multiply:Float(PrimaryOperand=UE.Power:Float(PrimaryOperand=^Fort.Vehicle.NormalSpeed) {also {defaults {AdditionalOperands=3}}}.Out) {also {defaults {AdditionalOperands=4}}}.Out
    Value 0=UE.Subtract:Float(AdditionalOperands=UE.Multiply:Float(PrimaryOperand=@n4.Out) {also {defaults {AdditionalOperands=2}}}.Out) {also {defaults {PrimaryOperand=1}}}.Out
    Set 1=@TriggerOnce_0.Trigger Out
    Set 0=@TriggerOnce_1.Trigger Out
   )
   n6 = MapRange.MapRange:Float(In=UE.InterpTo:Audio(Target=@TriggerRouteTriggerRouteFloat2_1.Value).Value) {also {defaults {Out Range A=-16, Out Range B=18}}}
   n7 = VariableAccessor.Float(Variable=VariableMutator.Float(Value=UE.Add:Float(AdditionalOperands=@n3.Value, PrimaryOperand=@n6.Out Value).Out, Variable=InitVariable.Float().Variable).Variable)
   n8 = VariableAccessor.Float(Variable=@n7.Variable)
   WavePlayer_0 = UE.Wave Player:Mono(Play=@n1.Activate, Stop=@n1.Deactivate, Pitch Shift=@n8.Value) {also {
     defaults {Loop="<MetaSoundBuilderResult.SUCCEEDED: 0>", Wave Asset=&/EngineAudio_Future_HighSpeed_Loop}
     version [1,1]
    }}
   WavePlayer_1 = UE.Wave Player:Mono(Play=^UE.Source.OnPlay, Pitch Shift=VariableAccessor.Float(Variable=@n8.Variable).Value) {also {
     defaults {Loop="<MetaSoundBuilderResult.SUCCEEDED: 0>", Wave Asset=&/EngineAudio_Future_LowSpeed_Loop}
     version [1,1]
    }}
   TriggerAnyTriggerAny2_1 = "TriggerAny.Trigger Any (2)"(In 1=@n2.Deactivate)
   TriggerAnyTriggerAny2_2 = "TriggerAny.Trigger Any (2)"(In 1=^UE.Source.OnPlay, In 0=@n2.Activate)
   WavePlayer_2 = UE.Wave Player:Mono(Stop=@TriggerAnyTriggerAny2_1.Out, Play=@TriggerAnyTriggerAny2_2.Out, Pitch Shift=@n7.Value) {also {
     defaults {Loop="<MetaSoundBuilderResult.SUCCEEDED: 0>", Wave Asset=&/EngineAudio_Future_MediumSpeed_Loop}
     version [1,1]
    }}
   AudioMixerAudioMixerMono3_0 = "AudioMixer.Audio Mixer (Mono, 3)"(
    Gain 1=UE.Reroute(Value=UE.Reroute(Value=@n1.Volume).Value).Value
    Gain 0=UE.Reroute(Value=UE.Reroute(Value=@n2.Volume).Value).Value
    In 1=@WavePlayer_0.Out Mono
    In 2=@WavePlayer_1.Out Mono
    In 0=@WavePlayer_2.Out Mono
   )
   DecibelstoLinearGain_2 = UE.Decibels to Linear Gain:Float(Decibels=VariableAccessor.Float(Variable=InitVariable.Float().Variable).Value)
   DecibelstoLinearGain_0 = UE.Decibels to Linear Gain:Float()
   DecibelstoLinearGain_3 = UE.Decibels to Linear Gain:Float() {also {defaults {Decibels=1.5}}}
   TriggerOnValueChange_1 = UE.Trigger On Value Change:Int32(Value=Convert.Int32:Enum:FortAudioAffiliation("Enum:FortAudioAffiliation"=^Character.Affiliation).Int32)
   TriggerAnyTriggerAny2_0 = "TriggerAny.Trigger Any (2)"(In 0=^UE.Source.OnPlay, In 1=@TriggerOnValueChange_1.Trigger)
   n9 = &TriggerCompareAffiliation(A=^Character.Affiliation, Compare=@TriggerAnyTriggerAny2_0.Out)
   TriggerRouteTriggerRouteFloat2_0 = "TriggerRoute.Trigger Route (Float, 2):Float"(Value 0=@DecibelstoLinearGain_0.Linear Gain, Value 1=@DecibelstoLinearGain_3.Linear Gain, Set 1=@n9.False, Set 0=@n9.True)
   n10 = UE.Multiply:Float(
    PrimaryOperand=@DecibelstoLinearGain_2.Linear Gain
    AdditionalOperands=UE.InterpTo:Audio(Target=@TriggerRouteTriggerRouteFloat2_0.Value).Value
   )
   DecibelstoLinearGain_1 = UE.Decibels to Linear Gain:Float() {also {defaults {Decibels=-8}}}
   n11 = /FortMetasound/Patches/MapRange_DecibelOutput(In=^Fort.Vehicle.NormalSpeed) {also {defaults {Out Range A=-10}}}
   WavePlayer_3 = UE.Wave Player:Stereo(
    Play=UE.TriggerOnThreshold:Float(In=^Fort.Vehicle.BonusSpeed) {also {defaults {Threshold=0.1}}}.Out
    Stop=UE.TriggerOnThreshold:Float(In=^Fort.Vehicle.BonusSpeed) {also {defaults {Threshold=0.1, Type=1}}}.Out
   ) {also {
     defaults {Loop="<MetaSoundBuilderResult.SUCCEEDED: 0>", Wave Asset=&/EngineAudio_Future_Supersonic_Loop}
     version [1,1]
    }}
   AudioMixerAudioMixerStereo2_0 = "AudioMixer.Audio Mixer (Stereo, 2)"(
    In 0 L=@AudioMixerAudioMixerMono3_0.Out
    In 0 R=@AudioMixerAudioMixerMono3_0.Out
    Gain 1=UE.Multiply:Float(
     PrimaryOperand=@n10.Out
     AdditionalOperands=UE.Multiply:Float(AdditionalOperands=@DecibelstoLinearGain_1.Linear Gain, PrimaryOperand=@n11.Out Value).Out
    ).Out
    Gain 0=@n10.Out
    In 1 L=@WavePlayer_3.Out Left
    In 1 R=@WavePlayer_3.Out Right
   )
  ]
  links [
   {dst="@UE.OutputFormat.Stereo.Audio:0", src=AudioMixerAudioMixerStereo2_0.Out L}
   {dst="@UE.OutputFormat.Stereo.Audio:1", src=AudioMixerAudioMixerStereo2_0.Out R}
  ]
 }]
properties {
 assetImportData=&self:AssetImportData
 attenuationSettings=&Vehicle_Engine_Att
 audioPropertiesSheet=&APS_Cosmetic_Vehicle_Engine
 bLooping=true
 duration=10000
 maxDistance=12000
 numChannels=2
 outputFormat=Stereo
 priority=80
 sampleRate=48000
 soundClassObject=&Fort_Vehicles_Engine
 concurrencySet [&Conc_Veh_EngineAudio_LowSpec, &Conc_Veh_EngineAudio_LowSpec_Actor]
 modulationSettings {
  highpassRouting=Union
  lowpassRouting=Union
  pitchRouting=Union
  volumeRouting=Union
  highpassModulationDestination {bEnableModulation=true, modulators [&PP_FortVehicles_Engine_HPF]}
  lowpassModulationDestination {bEnableModulation=true, modulators [&Patch_FortVehicles_Engine_LPF]}
  pitchModulationDestination {bEnableModulation=true, modulators [&DM_Bus_Vehicle_Engine_Pitch]}
  volumeModulationDestination {
   bEnableModulation=true
   modulators [&DM_VehicleEnginePatch, &Patch_FortVehicles_Engine_Vol]
  }
 }
}
